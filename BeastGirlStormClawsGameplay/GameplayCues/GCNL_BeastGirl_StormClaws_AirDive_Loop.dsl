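; Looping gameplay cue for Beast Girl's StormClaws air dive that drives a charging material effect on her head mesh.
; On looping start it casts the target to the Beast Girl pawn and gets its Head skeletal mesh.
; It sets IsActive on the materials and plays a Charging timeline into PostEM.
; An async tag wait clears IsActive when the StormSlam cooldown tag is added to the pawn.
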(event EventOnLoopingStart(EitherVFX-preferthisversion)
  (params (MyTarget :type Actor%20Object%20Reference) (Parameters :type Gameplay%20Cue%20Parameters%20Structure%20%28by%20ref%29) (ParticleComponents :type Array%20of%20FXSystem%20Component%20Object%20References) (AudioComponents :type Array%20of%20Audio%20Component%20Object%20References))
  (bind _bp_beast_girl_pawn (Variables|Default|GetBP%20Beast%20Girl%20Pawn))
  (bind _returnvalue (Customization|GetSkeletalMeshforPartType _bp_beast_girl_pawn "Head" :!class /Script/FortniteGame.FortPlayerPawn))
  ; Tech Art
  (bind _asbp_beast_girl_pawn (Utilities|Casting|CastToBP_BeastGirlPawn MyTarget))
  (Variables|Default|SetBP%20Beast%20Girl%20Pawn _asbp_beast_girl_pawn)
  (Ability|Tasks|WaitGameplayTagAddToActor _bp_beast_girl_pawn "(TagName=\"Granted.BeastGirl.StormClaws.StormSlam.Cooldown\")"
    (:then
      (Rendering|Material|SetScalarParameterValueonMaterials _returnvalue "IsActive" 1.0)
      (bind (_direction _charging) (Timeline|Charging PlayFromStart)
        (:Update
          (Rendering|Material|SetScalarParameterValueonMaterials _returnvalue "PostEM" _charging))))
    (:Added
      (Rendering|Material|SetScalarParameterValueonMaterials _returnvalue "IsActive" 0.0))))

(event Collision|EventActorBeginOverlap
  (params (OtherActor :type Actor%20Object%20Reference)))

(event EventTick
  (params (DeltaSeconds :type Float%20%28single-precision%29)))

(fn ConstructionScript)
