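; Looping gameplay cue for Beast Girl that shows when her Storm Slam ability is ready or on cooldown.
; Application and looping-start events cache their particle components, and application also sets a five-second timer to start watching the cooldown tag count on the pawn.
; When the tag appears, the looping particle spawn rate drops to zero.
; When it clears, the application effect reactivates, spawn rate is restored, an attached Niagara spawn runs, and both changes print debug text.
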(event EventOnApplication(EitherVFX-preferthisone)
  (params (MyTarget :type Actor%20Object%20Reference) (Parameters :type Gameplay%20Cue%20Parameters%20Structure%20%28by%20ref%29) (ParticleComponents :type Array%20of%20FXSystem%20Component%20Object%20References) (AudioComponents :type Array%20of%20Audio%20Component%20Object%20References) (BurstCameraShakeInstance :type Camera%20Shake%20Base%20Object%20Reference) (BurstDecalInstance :type Decal%20Actor%20Object%20Reference))
  (for _array_element ParticleComponents
    (Variables|Default|SetOn%20Application%20System _array_element)
    (Utilities|Time|SetTimerbyEvent (delegate Custom|CheckPawn) 5.0)))

(event Custom|CheckPawn
  (bind _asbp_beast_girl_pawn (Utilities|Casting|CastToBP_BeastGirlPawn (stale EventOnApplication(EitherVFX-preferthisone) MyTarget)))
  (bind (_asynctaskproxy _tagcount) (Ability|Tasks|WaitGameplayTagCountChangedOnActor _asbp_beast_girl_pawn (Variables|Default|GetStormSlamCooldownTag))
    (:TagCountChanged
      (if (>= _tagcount 1)
        (Development|PrintString 0 "The Storm Slam is ON COOLDOWN")
        (Variables|Default|SetIsOnCooldown true)
        (Effects|Components|ParticleSystem|SetFloatParameter (Variables|Default|GetLooping%20Particle%20System) "Spawn Rate" 0.0)
        (else
          (Development|PrintString 0 "The Storm Slam is OFF COOLDOWN")
          (Variables|Default|SetIsOnCooldown false)
          (bind _returnvalue (Niagara|SpawnSystemAttached 0 (Variables|/Script/Engine.Actor|GetRootComponent _asbp_beast_girl_pawn) "None" "0, 0, 0" "0, 0, 0" "KeepRelativeOffset" false false))
          (Components|Activation|Activate (Variables|Default|GetOn%20Application%20System) true)
          (Effects|Components|ParticleSystem|SetFloatParameter (Variables|Default|GetLooping%20Particle%20System) "Spawn Rate" 1.0))))))

(event EventOnLoopingStart(EitherVFX-preferthisversion)
  (params (MyTarget :type Actor%20Object%20Reference) (Parameters :type Gameplay%20Cue%20Parameters%20Structure%20%28by%20ref%29) (ParticleComponents :type Array%20of%20FXSystem%20Component%20Object%20References) (AudioComponents :type Array%20of%20Audio%20Component%20Object%20References))
  (for _array_element ParticleComponents
    (Variables|Default|SetLooping%20Particle%20System _array_element)))

(event Collision|EventActorBeginOverlap
  (params (OtherActor :type Actor%20Object%20Reference)))

(event EventTick
  (params (DeltaSeconds :type Float%20%28single-precision%29)))

(fn ConstructionScript)
