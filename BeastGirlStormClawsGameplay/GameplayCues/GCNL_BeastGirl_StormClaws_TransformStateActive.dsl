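; Looping gameplay cue for Beast Girl Storm Claws' active transformation state, which puts a Niagara camera lens effect on the target's player.
; On application it manually adds the lens effect, seeds its Sprint parameter once from the pawn and binds DisplayHitNotify to flash Pulse for a frame.
; It also starts a wait for the reverse-transformation event, with no handler attached here.
; Removal ends that wait and deactivates the Niagara component.
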(event EventOnApplication(EitherVFX-preferthisone)
  (params (MyTarget :type Actor%20Object%20Reference) (Parameters :type Gameplay%20Cue%20Parameters%20Structure%20%28by%20ref%29) (ParticleComponents :type Array%20of%20FXSystem%20Component%20Object%20References) (AudioComponents :type Array%20of%20Audio%20Component%20Object%20References) (BurstCameraShakeInstance :type Camera%20Shake%20Base%20Object%20Reference) (BurstDecalInstance :type Decal%20Actor%20Object%20Reference))
  (bind _target (Variables|Default|GetTarget))
  (bind _returnvalue (Fortnite|GetFortPlayerControllerfromActor _target))
  (bind _output_get (Variables|Default|SetTarget MyTarget))
  (bind _asynctaskproxy (Ability|Async|WaitGameplayEventToActor _output_get "(TagName=\"BeastGirl.Event.BeginReverseTransformation\")" true))
  (Variables|Default|SetWaitEventAction _asynctaskproxy)
  ; This is what we need to apply; We're applying the camera BP manually since we have to modify/update it later (tieing into health for visuals)
  (Utilities|IsValid _returnvalue
    (:Is%20Valid
      (bind _output_get_2 (Variables|Default|SetCachedController _returnvalue))
      (bind _returnvalue_2 (Camera|AddGenericCameraLensEffect (Variables|/Script/Engine.PlayerController|GetPlayerCameraManager _output_get_2) (Variables|Default|GetLooping%20Lens%20Effect%20BP)))
      (bind _output_get_3 (Variables|Default|SetLens%20Effect%20Interface _returnvalue_2))
      (bind _asniagara_lens_effect_base (Utilities|Casting|CastToNiagaraLensEffectBase _output_get_3))
      (bind _output_get_4 (Variables|Default|SetNiagara%20Lens%20Effect _asniagara_lens_effect_base))
      (Utilities|FlowControl|Sequence
        (:then_0
          (bind _asbp_beast_girl_pawn (Utilities|Casting|CastToBP_BeastGirlPawn (Variables|Default|GetTarget)))
          (EventDispatchers|BindEventtoDisplayHitNotify _asbp_beast_girl_pawn (EventDispatchers|CreateEvent :!event Custom|Hit%20Player))
          (Variables|Default|SetIs%20Sprinting (PawnStats|IsSprinting _asbp_beast_girl_pawn)))
        (:then_1
          (Effects|Components|ParticleSystem|SetBoolParameter (Variables|/Script/Niagara.NiagaraActor|GetNiagaraComponent _output_get_4) "Sprint" (Variables|Default|GetIs%20Sprinting)))))
    (:Is%20Not%20Valid
      (Development|PrintString 0 "no valid "))))

(event Custom|Hit%20Player
  (params (Damage :type Float%20%28double-precision%29) (Critical :type Boolean) (Fatal :type Boolean) (Shield :type Boolean) (HitPawn :type Actor%20Object%20Reference) (HitLocation :type Vector) (Tags :type Gameplay%20Tag%20Container%20Structure))
  (bind _niagara_lens_effect (Variables|Default|GetNiagara%20Lens%20Effect))
  (bind _niagaracomponent (Variables|/Script/Niagara.NiagaraActor|GetNiagaraComponent _niagara_lens_effect))
  (Effects|Components|ParticleSystem|SetBoolParameter _niagaracomponent "Pulse" true)
  (Utilities|FlowControl|Delay 0)
  (Effects|Components|ParticleSystem|SetBoolParameter _niagaracomponent "Pulse" false))

(event EventOnRemoval(EitherVFX-preferthisversion)
  (params (MyTarget :type Actor%20Object%20Reference) (Parameters :type Gameplay%20Cue%20Parameters%20Structure%20%28by%20ref%29) (ParticleComponents :type Array%20of%20FXSystem%20Component%20Object%20References) (AudioComponents :type Array%20of%20Audio%20Component%20Object%20References) (BurstCameraShakeInstance :type Camera%20Shake%20Base%20Object%20Reference) (BurstDecalInstance :type Decal%20Actor%20Object%20Reference))
  (bind _niagara_lens_effect (Variables|Default|GetNiagara%20Lens%20Effect))
  (Ability|Async|EndAction (Variables|Default|GetWaitEventAction))
  ; Ensuring End of niagara system if the cue is destroyed
  (Utilities|IsValid _niagara_lens_effect
    (:Is%20Valid
      (Components|Activation|Deactivate (Variables|/Script/Niagara.NiagaraActor|GetNiagaraComponent _niagara_lens_effect)))))

(fn ConstructionScript)
