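; Looping gameplay cue notify for the Beast Girl pawn that plays a timed color shift on the character, named for a time-ending state.
; On application it casts the target to BP_BeastGirlPawn and caches its Head and Body skinned mesh components.
; A timeline then writes its value to custom primitive data slot 0 on each valid cached mesh.
; When the timeline finishes, the cue ends and recycles itself.
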(event EventOnApplication(EitherVFX-preferthisone)
  (params (MyTarget :type Actor%20Object%20Reference) (Parameters :type Gameplay%20Cue%20Parameters%20Structure%20%28by%20ref%29) (ParticleComponents :type Array%20of%20FXSystem%20Component%20Object%20References) (AudioComponents :type Array%20of%20Audio%20Component%20Object%20References) (BurstCameraShakeInstance :type Camera%20Shake%20Base%20Object%20Reference) (BurstDecalInstance :type Decal%20Actor%20Object%20Reference))
  (bind _asbp_beast_girl_pawn (Utilities|Casting|CastToBP_BeastGirlPawn MyTarget))
  (bind _output_get (Variables|Default|SetCachedMeshes (Customization|GetSkinnedMeshComponentsforPartType _asbp_beast_girl_pawn "Head")))
  (Utilities|Array|AppendArray _output_get (Customization|GetSkinnedMeshComponentsforPartType _asbp_beast_girl_pawn "Body"))
  (bind (_direction _newtrack_0) (Timeline|ColorShiftTimeline Play)
    (:Update
      (for _array_element (Variables|Default|GetCachedMeshes)
        (Utilities|IsValid _array_element
          (:Is%20Valid
            (Rendering|Material|SetCustomPrimitiveDataFloat _array_element 0 _newtrack_0)))))
    (:Finished
      (GameplayCueNotify|End%28Recycle%29GameplayCue))))

(fn ConstructionScript)
