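; Gameplay cue notify for the Beast Girl (Storm Claws) de-transform, adding a burst sound on the target pawn.
; On burst it runs the parent logic, casts the target to a Fortnite player pawn, and plays a medium wing-flap sound through the Sound Library subsystem, using the "head" socket.
; Volume uses the local-or-enemy level for locally viewed and enemy pawns, and the friendly-or-neutral level for other pawns with those affiliations.
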(fn OnBurst(EitherVFX-preferthisversion)
  (params (MyTarget :type Actor%20Object%20Reference) (Parameters :type Gameplay%20Cue%20Parameters%20Structure%20%28by%20ref%29) (ParticleComponents :type Array%20of%20FXSystem%20Component%20Object%20References) (AudioComponents :type Array%20of%20Audio%20Component%20Object%20References) (BurstCameraShakeInstance :type Camera%20Shake%20Base%20Object%20Reference) (BurstDecalInstance :type Decal%20Actor%20Object%20Reference))
  (|CallParentFunction MyTarget Parameters ParticleComponents AudioComponents BurstCameraShakeInstance BurstDecalInstance)
  (bind _asfort_player_pawn_athena (Utilities|Casting|CastToFortPlayerPawnAthena MyTarget))
  (bind _output_get (LocalVariables|SetTarget%20Pawn _asfort_player_pawn_athena))
  (SoundLibrary|PlaySoundSimple (WorldSubsystems|GetSoundLibrarySubsystem (LocalVariables|GetTarget%20Pawn)) (LocalVariables|GetTarget%20Pawn) (Utilities|Struct|MakeSimpleSoundSettings "(TagName=\"SoundLibrary.Irwin.Action.Flying.WingFlap.Medium\")" true false true false (Math|Float|SelectFloat (LocalVariables|GetLocal%20or%20Enemy%20Volume) (Math|Float|SelectFloat (LocalVariables|GetFriendly%20or%20Neutral%20Volume) (LocalVariables|GetLocal%20or%20Enemy%20Volume) (or (Utilities|Enum|Equal(Enum) (Team|GetTeamAffiliationforActortoLocalPlayer _output_get) "Friendly") (Utilities|Enum|Equal(Enum) (Team|GetTeamAffiliationforActortoLocalPlayer _output_get) "Neutral"))) (Pawn|IsLocallyViewed _output_get)) 1.0 0.0 0 0 "head" 0 "SnapToTargetIncludingScale") :!class /Script/SoundLibrary.SoundLibrarySubsystem)
  (return))
